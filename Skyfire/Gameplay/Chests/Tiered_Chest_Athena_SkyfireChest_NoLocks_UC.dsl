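; Skyfire tiered loot chest without locks, with idle glow, light and ambient sound cues until searched.
; Looting or searching clears those cues, drives a timeline-fed CODE_Timestamp material parameter for opening, spawns faction Niagara effects and lets players build over it.
; On mobile, dynamic materials and timelines drive interact wiggle, outlines and a scaling icon.
; For Outpost player controllers, only the owner can interact with or destroy it.
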(event EventBeginPlay
  ;[1 1 1] Begin Play
  (Networking|CanExecuteCosmeticEvents
    (:True
      (CallFunction|CreateMobileMIDs)
      (if (Variables|Container|GetbAlreadySearched)
        (CallFunction|SetLightVisibility false)
        (Audio|Components|Audio|Stop (Variables|Default|GetChest_Ambient_Sound))
        (else
          (Rendering|SetVisibility (Variables|Default|GetFaction_Crate_Glows) true false)
          (CallFunction|SetLightVisibility true)
          (CallFunction|StartAudiovisualIndicator (Variables|Default|GetChest_Ambient_Sound) "Chest" (Variables|Default|GetChestChimeVisualUpdate)))))))

(event Custom|Mobile_OnInteract
  ;[1 1 1] Mobile Wiggle on Interact
  (Networking|CanExecuteCosmeticEvents
    (:True
      (bind (_direction _lerp) (Timeline|MobileOnInteractTL Play)
        (:Update
          (for _array_element (Variables|Default|GetMIDs)
            (Rendering|Material|SetScalarParameterValue _array_element "WiggleOnInteractAmount" (Math|Float|Lerp 0.0 (Variables|Default|GetMobileWiggleAmount) _lerp) :!class /Script/Engine.MaterialInstanceDynamic)))))))

(event Custom|Mobile_Outline_On
  ;[1 1 1] Mobile Outline
  (for _array_element (Variables|Default|GetMIDs)
    (Rendering|Material|SetScalarParameterValue _array_element "Outline_Control" 1.0 :!class /Script/Engine.MaterialInstanceDynamic)
    (Rendering|Material|SetScalarParameterValue _array_element "Pulse_Control" 1.0 :!class /Script/Engine.MaterialInstanceDynamic)
    (Rendering|Material|SetVectorParameterValue _array_element "Color" "(R=1.000000,G=1.000000,B=1.000000,A=1.000000)" :!class /Script/Engine.MaterialInstanceDynamic)))

(event Container|EventEnableBacchusHighlight
  ;[1 1 1] Mobile highlighting
  (if (not (Variables|Container|GetbAlreadySearched))
    (Networking|CanExecuteCosmeticEvents
      (:True
        (CallFunction|SetVisibleMobileInteractIcon true)
        (bind (_direction _lerpobject _lerpinteactoicon) (Timeline|MobileSelectedTL Play)
          (:Update
            (bind _mobileinteracticon (Variables|Default|GetMobileInteractIcon)
              (:then
                (Transformation|SetRelativeScale3D _mobileinteracticon (Math|Vector|Lerp(Vector) "0, 0, 0" (Variables|Default|GetMobileInteractIconScale) _lerpinteactoicon)))
              (:else)))
          (:Finished
            (switch Utilities|FlowControl|Switch|SwitchonETimelineDirection _direction
              (:Backward
                (CallFunction|SetVisibleMobileInteractIcon false)))))))))

(event Custom|Mobile_OnDisconnect
  ;[1 1 1] Mobile Wiggle on Interact
  (Networking|CanExecuteCosmeticEvents
    (:True
      (Timeline|MobileOnInteractTL Reverse))))

(event Custom|Mobile_Outline_Off
  ;[1 1 1] Mobile Outline
  (for _array_element (Variables|Default|GetMIDs)
    (Rendering|Material|SetScalarParameterValue _array_element "Outline_Control" 0.0 :!class /Script/Engine.MaterialInstanceDynamic)
    (Rendering|Material|SetScalarParameterValue _array_element "Pulse_Control" 0.0 :!class /Script/Engine.MaterialInstanceDynamic)))

(event Container|EventDisableBacchusHighlight
  ;[1 1 1] Mobile highlighting
  (if (TouchPlatform|IsMobileGame)
    (Networking|CanExecuteCosmeticEvents
      (:True
        (Timeline|MobileSelectedTL Reverse)))))

(event EventOnBeginSearch
  ;[1 1 1] Mobile highlighting
  (if (Variables|Container|GetbAlreadySearched)
    _
    (elif (TouchPlatform|IsMobileGame)
      (Networking|CanExecuteCosmeticEvents
        (:True
          (Timeline|MobileSelectedTL Reverse))))))

(event EventOnLoot
  (Utilities|FlowControl|Sequence
    (:then_0
      (CallFunction|PlayOpeningTimeLine)
      (Rendering|SetVisibility (Variables|Default|GetFaction_Crate_Glows) false false)
      (Components|DestroyComponent (Variables|Default|GetP_NS_Faction_Chest_Idle))
      (CallFunction|SetLightVisibility false)
      (Audio|Components|Audio|FadeOut (Variables|Default|GetChest_Ambient_Sound) 0.5 0.0 "Linear" :!class /Script/Engine.AudioComponent)
      (CallFunction|StopAudiovisualIndicator)
      (Variables|Building|SetbDestroyOnPlayerBuildingPlacement true) ; Allows players to build over the opened chest mesh (but not when there is loot in there)
      (CallFunction|CleanupWiggleMIDs)
      (Components|Timeline|Stop (Variables|Tiered_Chest_Athena_SkyfireChest_NoLocks_UC|GetMobileSelectedTL))
      (CallFunction|SetVisibleMobileInteractIcon false)
      (Rendering|SetRenderCustomDepth (Variables|Debug|GetStaticMeshComponent) false))
    (:then_1
      ;[1 0.4913 0.5813] Niagara fx
      (bind _returnvalue (Niagara|SpawnSystematLocation "/Game/Effects/Fort_Effects/Gameplay/Chests/NS_Chest_Factions_Open.NS_Chest_Factions_Open" (Transformation|GetActorLocation) (Math|Rotator|MakeRotator (.roll (Transformation|GetActorRotation)) (+ (.pitch (Transformation|GetActorRotation)) (Variables|Default|GetOpenFX_Rotation)) (.yaw (Transformation|GetActorRotation))) "1.000000,1.000000,1.000000" true true "AutoRelease"))
      (Variables|Default|SetLoot_Effect_Niagara _returnvalue)
      (Fortnite|GetRarityData (Variables|Container|GetHighestRarity))
      (Utilities|FlowControl|DoOnce false
        (:Completed
          (Audio|Components|Audio|Play (Variables|Default|GetChest_Open_Sound) 0.0))))))

(event EventOnSetSearched
  (Utilities|FlowControl|Sequence
    (:then_0
      (CallFunction|PlayOpeningTimeLine)
      (Rendering|SetVisibility (Variables|Default|GetFaction_Crate_Glows) false false)
      (Components|DestroyComponent (Variables|Default|GetP_NS_Faction_Chest_Idle))
      (CallFunction|SetLightVisibility false)
      (Audio|Components|Audio|FadeOut (Variables|Default|GetChest_Ambient_Sound) 0.5 0.0 "Linear" :!class /Script/Engine.AudioComponent)
      (CallFunction|StopAudiovisualIndicator)
      (Variables|Building|SetbDestroyOnPlayerBuildingPlacement true) ; Allows players to build over the opened chest mesh (but not when there is loot in there)
      (CallFunction|CleanupWiggleMIDs)
      (Components|Timeline|Stop (Variables|Tiered_Chest_Athena_SkyfireChest_NoLocks_UC|GetMobileSelectedTL))
      (CallFunction|SetVisibleMobileInteractIcon false)
      (Rendering|SetRenderCustomDepth (Variables|Debug|GetStaticMeshComponent) false))
    (:then_1
      ;[1 0.4913 0.5813] Niagara fx
      (bind _returnvalue (Niagara|SpawnSystematLocation "/Game/Effects/Fort_Effects/Gameplay/Chests/NS_Chest_Factions_Open.NS_Chest_Factions_Open" (Transformation|GetActorLocation) (Math|Rotator|MakeRotator (.roll (Transformation|GetActorRotation)) (+ (.pitch (Transformation|GetActorRotation)) (Variables|Default|GetOpenFX_Rotation)) (.yaw (Transformation|GetActorRotation))) "1.000000,1.000000,1.000000" true true "AutoRelease"))
      (Variables|Default|SetLoot_Effect_Niagara _returnvalue)
      (Fortnite|GetRarityData (Variables|Container|GetHighestRarity))
      (Utilities|FlowControl|DoOnce false
        (:Completed
          (Audio|Components|Audio|Play (Variables|Default|GetChest_Open_Sound) 0.0))))))

(event EventDestroyed
  (CallFunction|PlayOpeningTimeLine)
  (Rendering|SetVisibility (Variables|Default|GetFaction_Crate_Glows) false false)
  (Components|DestroyComponent (Variables|Default|GetP_NS_Faction_Chest_Idle))
  (CallFunction|SetLightVisibility false)
  (Audio|Components|Audio|FadeOut (Variables|Default|GetChest_Ambient_Sound) 0.5 0.0 "Linear" :!class /Script/Engine.AudioComponent)
  (CallFunction|StopAudiovisualIndicator)
  (Variables|Building|SetbDestroyOnPlayerBuildingPlacement true) ; Allows players to build over the opened chest mesh (but not when there is loot in there)
  (CallFunction|CleanupWiggleMIDs)
  (Components|Timeline|Stop (Variables|Tiered_Chest_Athena_SkyfireChest_NoLocks_UC|GetMobileSelectedTL))
  (CallFunction|SetVisibleMobileInteractIcon false)
  (Rendering|SetRenderCustomDepth (Variables|Debug|GetStaticMeshComponent) false))

(event Interaction|EventBlueprintOnInteract
  (params (InteractingPawn :type Fort%20Pawn%20Object%20Reference) (InteractionBeingAttempted :type EInteractionBeingAttempted%20Enum :default "FirstInteraction")))

(event Custom|PlayOpeningTimeLine
  (bind (_direction _time) (Timeline|Timeline_0 PlayFromStart)
    (:Update
      (Rendering|Material|SetScalarParameterValueonMaterials (Variables|Debug|GetStaticMeshComponent) "CODE_Timestamp" _time))))

(event EventOnDeathPlayEffects
  (params (Damage :type Float%20%28single-precision%29) (DamageTags :type Gameplay%20Tag%20Container%20Structure%20%28by%20ref%29) (Momentum :type Vector :default "0, 0, 0") (HitInfo :type Hit%20Result%20Structure%20%28by%20ref%29) (InstigatedBy :type Fort%20Pawn%20Object%20Reference) (DamageCauser :type Actor%20Object%20Reference) (EffectContext :type Gameplay%20Effect%20Context%20Handle%20Structure))
  (Audio|Components|Audio|FadeOut (Variables|Default|GetChest_Ambient_Sound) 0.5 0.0 "Linear" :!class /Script/Engine.AudioComponent))

(fn BlueprintCanInteract
  (params (InteractingPawn :type Fort%20Pawn%20Object%20Reference) (InteractionBeingAttempted :type EInteractionBeingAttempted%20Enum :default "FirstInteraction") (InteractionType :type TInteractionType%20Enum :default "IT_NoInteraction"))
  (bind _asfort_player_controller_outpost (Utilities|Casting|CastToFortPlayerControllerOutpost (Pawn|GetController InteractingPawn))
    (:then
      (return (Outpost|IsOutpostOwner _asfort_player_controller_outpost)))
    (:CastFailed
      (return true))))

(fn CleanupWiggleMIDs
  (if (TouchPlatform|IsMobileGame)
    (Components|Timeline|Stop (Variables|Tiered_Chest_Athena_SkyfireChest_NoLocks_UC|GetMobileOnInteractTL))
    (for _array_element (Variables|Default|GetMIDs)
      (Rendering|Material|SetScalarParameterValue _array_element "WiggleOnInteractAmount" 0.0 :!class /Script/Engine.MaterialInstanceDynamic))))

(fn CreateMobileMIDs
  (if (TouchPlatform|IsMobileGame)
    (for (_ _array_index) (Rendering|Material|GetMaterials (Variables|Debug|GetStaticMeshComponent))
      (bind _returnvalue (CallFunction|GetorCreateMIDOnDefaultMesh _array_index))
      (Utilities|Array|SetArrayElem (Variables|Default|GetMIDs) _array_index _returnvalue true))))

(fn GetMaxAudibleDistance
  (bind _chest_ambient_sound (Variables|Default|GetChest_Ambient_Sound))
  (bind _sound (Variables|/Script/Engine.AudioComponent|GetSound _chest_ambient_sound))
  (Utilities|IsValid _sound
    (:Is%20Valid
      (return (Variables|/Script/Engine.SoundBase|GetMaxDistance _sound)))
    (:Is%20Not%20Valid
      (return 0.0))))

(fn OnSetCustomDepthStencilValue
  (params (PrimComponents :type Array%20of%20Primitive%20Component%20Object%20References) (bUseCustomDepth :type Boolean) (StencilValue :type Integer))
  (for _array_element PrimComponents
    (Rendering|SetCustomDepthStencilValue _array_element (Variables|GetStencilValue))
    (if (Variables|GetbUseCustomDepth)
      (Rendering|SetRenderCustomDepth _array_element true)))
  (return true))

(fn SetLightVisibility
  (params (Visible :type Boolean))
  (Variables|Default|GetTreasureLight
    (:then
      (Rendering|SetVisibility (Variables|Default|GetTreasureLight) Visible false))
    (:else)))

(fn SetVisibleMobileInteractIcon
  (params (Visible :type Boolean))
  (bind (_returnvalue_roll _returnvalue_pitch _returnvalue_yaw) (Math|Rotator|BreakRotator (Transformation|GetActorRotation)))
  (if Visible
    ;[1 1 1] Create
    (bind _mobileinteracticon (Variables|Default|GetMobileInteractIcon)
      (:then
        (Rendering|SetVisibility _mobileinteracticon true false))
      (:else
        (Variables|Default|GetMobileInteractionMaterial
          (:then
            (bind _returnvalue (AddComponent|NODE_AddStaticMeshComponent-1 :RelativeTransform_Location (Variables|Default|GetMobileInteractIconLocation) :RelativeTransform_Rotation_Yaw (* _returnvalue_yaw -1.0) :RelativeTransform_Scale (Variables|Default|GetMobileInteractIconScale)))
            (bind _output_get (Variables|Default|SetMobileInteractIcon _returnvalue))
            (Rendering|Material|SetMaterial _output_get 0 (Variables|Default|GetMobileInteractionMaterial)))
          (:else))))
    (else
      ;[1 1 1] Destroy
      (bind _mobileinteracticon_2 (Variables|Default|GetMobileInteractIcon)
        (:then
          (Rendering|SetVisibility _mobileinteracticon_2 false false)
          (Components|DestroyComponent _mobileinteracticon_2))
        (:else)))))

(fn ShouldDie
  (params (Damage :type Float%20%28single-precision%29) (EventInstigator :type Controller%20Object%20Reference) (DamageCauser :type Actor%20Object%20Reference) (bIsTest :type Boolean) (EffectContext :type Gameplay%20Effect%20Context%20Handle%20Structure%20%28by%20ref%29))
  (bind _asfort_player_controller_outpost (Utilities|Casting|CastToFortPlayerControllerOutpost EventInstigator)
    (:then
      (return (Outpost|IsOutpostOwner _asfort_player_controller_outpost)))
    (:CastFailed
      (return true))))

(fn ConstructionScript
  (Networking|CanExecuteCosmeticEvents
    (:True
      (Utilities|EffectsQualityCheck 0
        (:Medium
          _)
        (:Low
          _))
      (Components|DestroyComponent (Variables|Default|GetTreasureLight)))))
